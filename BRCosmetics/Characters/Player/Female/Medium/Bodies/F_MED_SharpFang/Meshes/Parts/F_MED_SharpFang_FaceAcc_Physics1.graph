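# Physics asset for the female medium SharpFang character's face accessory part, setting up dangling hair and earring physics.
# Capsules on head, neck, clavicles and upper spine form the body, joined by angularly locked constraints except a free spine_05-to-spine_04 joint.
# A six-link back hair chain, two three-link front chains, single-body strands and both earrings hang from the head.
# These dynamic parts use small capsules or spheres with mostly tight swing and twist limits.
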PhysicsAsset /BRCosmetics/Characters/Player/Female/Medium/Bodies/F_MED_SharpFang/Meshes/Parts/F_MED_SharpFang_FaceAcc_Physics1
bodies {
 clavicle_l [{shapeType=Capsule, length=9, radius=7.25, shapeName=clavicle_l_capsule, center=<3.7866 0 -0.8159>, rotation=<rot 90 0 0>}]
 clavicle_r [{shapeType=Capsule, length=9, radius=7.25, shapeName=clavicle_l_capsule, center=<-3.7866 0 0.8159>, rotation=<rot 90 0 180>}]
 dyn_Earring_l [{
   shapeType=Capsule
   length=3.0482
   radius=1.612
   shapeName=dyn_Earring_l_capsule
   center=<3.8123 0.0318 -0.0647>
   rotation=<rot 0 -89.5103 -89.0385>
  }]
 dyn_Earring_r [{
   shapeType=Capsule
   length=2.5482
   radius=1.612
   shapeName=dyn_Earring_r_capsule
   center=<3.5648 -0.4409 0.061>
   rotation=<rot 0 82.9617 -89.0385>
  }]
 dyn_hair_back_01 [{shapeType=Capsule, length=0.25, radius=3.25, shapeName=dyn_hair_back_01_capsule, rotation=<rot 0 -90 0>}]
 dyn_hair_back_02 [{shapeType=Sphere, radius=3, shapeName=dyn_hair_back_02_sphere}]
 dyn_hair_back_03 [{shapeType=Sphere, radius=3, shapeName=dyn_hair_back_03_sphere}]
 dyn_hair_back_04 [{shapeType=Sphere, radius=3.25, shapeName=dyn_hair_back_04_sphere}]
 dyn_hair_back_05 [{shapeType=Sphere, radius=3.25, shapeName=dyn_hair_back_05_sphere, center=<0.1319 0.1143 0>}]
 dyn_hair_back_06 [{shapeType=Sphere, radius=3.25, shapeName=dyn_hair_back_06_sphere}]
 dyn_hair_back_dop [{shapeType=Sphere, radius=2, shapeName=dyn_hair_back_dop_sphere, center=<0.7936 1.0829 -0.0213>}]
 dyn_hair_front_A_01 [{shapeType=Capsule, length=3, radius=2, shapeName=dyn_hair_front_A_01_capsule, center=<3.5557 0 0>, rotation=<rot -90 4.9994 0>}]
 dyn_hair_front_A_02 [{shapeType=Sphere, radius=2, shapeName=dyn_hair_front_A_02_sphere, center=<1.6005 0.0991 -0.4025>}]
 dyn_hair_front_A_03 [{shapeType=Sphere, radius=2, shapeName=dyn_hair_front_A_03_sphere, center=<1.5832 -0.4683 -0.086>}]
 dyn_hair_front_B_01 [{shapeType=Sphere, radius=2, shapeName=dyn_hair_front_B_01_sphere, center=<1.2189 0.3842 -0.4117>}]
 dyn_hair_front_C_01 [{shapeType=Sphere, radius=2, shapeName=dyn_hair_front_C_01_sphere, center=<1.2366 0.4558 0.257>}]
 dyn_hair_front_D_01 [{
   shapeType=Capsule
   length=3.5581
   radius=1.4664
   shapeName=dyn_hair_front_D_01_capsule
   center=<2.3419 -0.0534 -0.031>
   rotation=<rot 0 -90.6299 -87.0374>
  }]
 dyn_hair_front_D_02 [{
   shapeType=Capsule
   length=3.0135
   radius=1.523
   shapeName=dyn_hair_front_D_02_capsule
   center=<3.2923 1.3924 -0.1157>
   rotation=<rot 0 115.8998 -87.4068>
  }]
 dyn_hair_front_D_03 [{
   shapeType=Capsule
   length=3.0532
   radius=0.7934
   shapeName=dyn_hair_front_D_03_capsule
   center=<3.4716 1.1796 -1.0092>
   rotation=<rot 0 -67.255 -71.2489>
  }]
 dyn_hair_side [{shapeType=Sphere, radius=2, shapeName=dyn_hair_side_sphere, center=<1.1352 0.0777 -0.7129>}]
 head [
  {
   shapeType=Capsule
   length=3.6921
   radius=8.2511
   shapeName=head_capsule
   center=<8.0445 1.141 0.4993>
   rotation=<rot 0.127 118.2702 -88.9607>
  }
  {shapeType=Capsule, length=5.25, radius=5.75, shapeName=head_capsule, center=<2.9208 3.8703 -0.9136>, rotation=<rot -76.8609 -161.6267 180>}
  {shapeType=Capsule, length=5.5, radius=5.5, shapeName=head_capsule, center=<2.5398 5.3068 -0.0063>, rotation=<rot -89.8569 179.9664 -180>}
  {shapeType=Capsule, length=5.25, radius=5.75, shapeName=head_capsule, center=<2.9208 3.8703 0.9136>, rotation=<rot 76.8609 -161.6267 180>}
 ]
 neck_01 [{shapeType=Capsule, length=3.2946, radius=5.2023, shapeName=neck_01_capsule, center=<3.2405 1.0114 -0.0096>, rotation=<rot 0 90 90>}]
 spine_04 [{shapeType=Capsule, length=9.5, radius=8, shapeName=spine_04_capsule, center=<3.1677 -2.5724 0>}]
 spine_05 [{shapeType=Capsule, length=25.51, radius=9.25, shapeName=spine_05_capsule, center=<11.1131 4.6312 0>, rotation=<rot 0 -90 0>}]
}
constraints [
 {
  bone1Name=head
  bone2Name=neck_01
  swing1LimitDegrees=45
  swing1Motion=Locked
  swing2LimitDegrees=45
  swing2Motion=Locked
  twistLimitDegrees=45
  twistMotion=Locked
 }
 {
  bone1Name=neck_01
  bone2Name=spine_05
  swing1LimitDegrees=45
  swing1Motion=Locked
  swing2LimitDegrees=45
  swing2Motion=Locked
  twistLimitDegrees=45
  twistMotion=Locked
 }
 {bone1Name=dyn_Earring_l, bone2Name=head, swing1LimitDegrees=10, swing2LimitDegrees=25, twistLimitDegrees=3}
 {bone1Name=dyn_hair_back_01, bone2Name=head, swing1LimitDegrees=5, swing2LimitDegrees=10, twistLimitDegrees=4}
 {bone1Name=dyn_hair_back_02, bone2Name=dyn_hair_back_01, swing1LimitDegrees=5, swing2LimitDegrees=3, twistLimitDegrees=4}
 {bone1Name=dyn_hair_back_03, bone2Name=dyn_hair_back_02, swing1LimitDegrees=3, swing2LimitDegrees=3, twistLimitDegrees=4}
 {bone1Name=dyn_hair_back_04, bone2Name=dyn_hair_back_03, swing1LimitDegrees=3, swing2LimitDegrees=3, twistLimitDegrees=4}
 {bone1Name=dyn_hair_back_05, bone2Name=dyn_hair_back_04, swing1LimitDegrees=3, swing2LimitDegrees=3, twistLimitDegrees=4}
 {bone1Name=dyn_hair_back_06, bone2Name=dyn_hair_back_05, swing1LimitDegrees=3, swing2LimitDegrees=6, twistLimitDegrees=4}
 {bone1Name=dyn_hair_back_dop, bone2Name=head, swing1LimitDegrees=10, swing2LimitDegrees=10, twistLimitDegrees=5}
 {bone1Name=dyn_hair_front_A_01, bone2Name=head, swing1LimitDegrees=20, swing2LimitDegrees=16, twistLimitDegrees=5}
 {bone1Name=dyn_hair_front_A_02, bone2Name=dyn_hair_front_A_01, swing1LimitDegrees=22, swing2LimitDegrees=22, twistLimitDegrees=5}
 {bone1Name=dyn_hair_front_A_03, bone2Name=dyn_hair_front_A_02, swing1LimitDegrees=16, swing2LimitDegrees=16, twistLimitDegrees=5}
 {bone1Name=dyn_hair_front_B_01, bone2Name=head, swing1LimitDegrees=10, swing2LimitDegrees=10, twistLimitDegrees=5}
 {bone1Name=dyn_hair_front_C_01, bone2Name=head, swing1LimitDegrees=10, swing2LimitDegrees=10, twistLimitDegrees=5}
 {bone1Name=dyn_hair_front_D_01, bone2Name=head, swing1LimitDegrees=40, swing2LimitDegrees=4.32, twistLimitDegrees=5}
 {bone1Name=dyn_hair_front_D_02, bone2Name=dyn_hair_front_D_01, swing1LimitDegrees=25, swing2LimitDegrees=9.16, twistLimitDegrees=5}
 {bone1Name=dyn_hair_front_D_03, bone2Name=dyn_hair_front_D_02, swing1LimitDegrees=25, swing2LimitDegrees=25, twistLimitDegrees=5}
 {bone1Name=dyn_hair_side, bone2Name=head, swing1LimitDegrees=10, swing2LimitDegrees=10, twistLimitDegrees=5}
 {
  bone1Name=clavicle_r
  bone2Name=spine_05
  swing1LimitDegrees=45
  swing1Motion=Locked
  swing2LimitDegrees=45
  swing2Motion=Locked
  twistLimitDegrees=45
  twistMotion=Locked
 }
 {bone1Name=dyn_Earring_r, bone2Name=head, swing1LimitDegrees=10, swing2LimitDegrees=25, twistLimitDegrees=3}
 {
  bone1Name=spine_05
  bone2Name=spine_04
  swing1LimitDegrees=45
  swing1Motion=Free
  swing2LimitDegrees=45
  swing2Motion=Free
  twistLimitDegrees=45
  twistMotion=Free
 }
 {
  bone1Name=clavicle_l
  bone2Name=spine_05
  swing1LimitDegrees=45
  swing1Motion=Locked
  swing2LimitDegrees=45
  swing2Motion=Locked
  twistLimitDegrees=45
  twistMotion=Locked
 }
]
