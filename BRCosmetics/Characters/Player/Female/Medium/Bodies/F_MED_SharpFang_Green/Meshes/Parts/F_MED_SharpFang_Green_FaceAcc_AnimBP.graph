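# Animation Blueprint for the SharpFang Green female FaceAcc mesh, copying the parent mesh pose and adding back-hair physics.
# A backpack flag blends between two RigidBody simulations, based at head or neck with separate physics assets.
# Additive bone offsets then correct the hair for capes, backpacks, pickaxe swings, skydive floating and head-to-spine rotation.
# The update event derives these inputs from backbling checks, the active montage and bone delta.
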AnimBlueprint /BRCosmetics/Characters/Player/Female/Medium/Bodies/F_MED_SharpFang_Green/Meshes/Parts/F_MED_SharpFang_Green_FaceAcc_AnimBP
target_skeleton=&/F_MED_SharpFang_Green_FaceAcc_Skeleton
graphs [
 AnimGraph {nodes [
   n1 = "Transform (Modify) Bone - Bone: dyn_hair_back_04"!(ComponentPose, bAlphaBoolEnabled:bool=True, Translation="0,0,0", Rotation="0.000,-12.740,-0.000", Scale="1,1,1") -> (Pose) {
    alphaInputType=Bool
    rotationMode=BMM_Additive
    rotationSpace=BCS_BoneSpace
    scaleMode=BMM_Additive
    scaleSpace=BCS_BoneSpace
    translationMode=BMM_Additive
    translationSpace=BCS_BoneSpace
    alphaBoolBlend {blendInTime=0.5, blendOption=Linear, blendOutTime=0.5}
    boneToModify {boneName=dyn_hair_back_04}
    rotation=<rot 0 -12.74 0>
   }
   n2 = "Transform (Modify) Bone - Bone: dyn_hair_back_05"!(ComponentPose=@n1, bAlphaBoolEnabled:bool=True, Translation="0.067,-2.826,0.000", Rotation="0.000,44.634,-0.000", Scale="1,1,1") -> (Pose) {
    alphaInputType=Bool
    rotationMode=BMM_Additive
    rotationSpace=BCS_BoneSpace
    scaleMode=BMM_Additive
    scaleSpace=BCS_BoneSpace
    translationMode=BMM_Additive
    translationSpace=BCS_BoneSpace
    alphaBoolBlend {blendInTime=0.5, blendOption=Linear, blendOutTime=0.5}
    boneToModify {boneName=dyn_hair_back_05}
    rotation=<rot 0 44.6341 0>
    translation=<0.067 -2.826>
   }
   "Transform (Modify) Bone - Bone: dyn_hair_back_06"!(ComponentPose=@n2, bAlphaBoolEnabled:bool=True, Translation="0,0,0", Rotation="0.000,31.343,-0.000", Scale="1,1,1") -> (Pose) {
    alphaInputType=Bool
    rotationMode=BMM_Additive
    rotationSpace=BCS_BoneSpace
    scaleMode=BMM_Additive
    scaleSpace=BCS_BoneSpace
    translationMode=BMM_Additive
    translationSpace=BCS_BoneSpace
    alphaBoolBlend {blendInTime=0.5, blendOption=Linear, blendOutTime=0.5}
    boneToModify {boneName=dyn_hair_back_06}
    rotation=<rot 0 31.343 0>
   }
   n3 = RigidBody!(
    ComponentPose=Local To Component(LocalPose=Copy Pose From Mesh() {bUseAttachedParent=true})
    Alpha
    OverrideWorldGravity:Vector="0,0,-980"
    ExternalForce
   ) -> (Pose) {
    bClampLinearTranslationLimitToRefPose=true
    bOverrideWorldGravity=true
    overridePhysicsAsset=/BRCosmetics/Characters/Player/Female/Medium/Bodies/F_MED_SharpFang/Meshes/Parts/F_MED_SharpFang_FaceAcc_Physics
    simulationSpace=BaseBoneSpace
    alphaBoolBlend {blendOption=Linear}
    baseBoneRef {boneName=head}
    componentAppliedLinearAccClamp=<2500 2500 2500>
    componentLinearAccScale=<0.15 0.4 0.15>
    componentLinearVelScale=<0.15 0.4 0.15>
    simSpaceSettings {
     maxAngularAcceleration=1000
     maxAngularVelocity=10
     maxLinearAcceleration=500
     maxLinearVelocity=250
     worldAlpha=1
     externalLinearDragV {z=2}
     externalLinearVelocity {z=2}
    }
   }
   n4 = Get bIsSkydiveFloating()
   n5 = "Transform (Modify) Bone - Bone: dyn_hair_back_02"!(ComponentPose=@n3, bAlphaBoolEnabled:bool=@n4, Translation="0,0,0", Rotation="-0.000,-41.162,-0.001", Scale="1,1,1") -> (Pose) {
    alphaInputType=Bool
    rotationMode=BMM_Additive
    rotationSpace=BCS_BoneSpace
    scaleMode=BMM_Additive
    scaleSpace=BCS_BoneSpace
    translationMode=BMM_Additive
    translationSpace=BCS_BoneSpace
    alphaBoolBlend {blendInTime=0.5, blendOption=Linear, blendOutTime=0.5}
    boneToModify {boneName=dyn_hair_back_02}
    rotation=<rot 0 -41.162 -0.001>
   }
   n6 = "Transform (Modify) Bone - Bone: dyn_hair_back_03"!(ComponentPose=@n5, bAlphaBoolEnabled:bool=@n4, Translation="0,0,0", Rotation="0.024,-29.920,0.040", Scale="1,1,1") -> (Pose) {
    alphaInputType=Bool
    rotationMode=BMM_Additive
    rotationSpace=BCS_BoneSpace
    scaleMode=BMM_Additive
    scaleSpace=BCS_BoneSpace
    translationMode=BMM_Additive
    translationSpace=BCS_BoneSpace
    alphaBoolBlend {blendInTime=0.5, blendOption=Linear, blendOutTime=0.5}
    boneToModify {boneName=dyn_hair_back_03}
    rotation=<rot 0.024 -29.92 0.04>
   }
   n7 = Component To Local(ComponentPose=@n6)
   n8 = RigidBody(ComponentPose=Local To Component(LocalPose=Copy Pose From Mesh() {bUseAttachedParent=true}), OverrideWorldGravity="0,0,-980") {
    bClampLinearTranslationLimitToRefPose=true
    bOverrideWorldGravity=true
    overridePhysicsAsset=/BRCosmetics/Characters/Player/Female/Medium/Bodies/F_MED_SharpFang/Meshes/Parts/F_MED_SharpFang_FaceAcc_Physics1
    simulationSpace=BaseBoneSpace
    alphaBoolBlend {blendOption=Linear}
    baseBoneRef {boneName=neck_01}
    componentLinearAccScale=<0.1 0.1 0.1>
    componentLinearVelScale=<0.1 0.1 0.1>
   }
   n9 = Component To Local(ComponentPose=@n8)
   n10 = Get Is Wearing Backpack() -> (Is Wearing Backpack:bool)
   n11 = Get Is Backbling A Cape() -> (Is Backbling A Cape:bool)
   n12 = "Transform (Modify) Bone - Bone: dyn_hair_back_02"(
    ComponentPose=Local To Component(LocalPose=Blend Poses by bool(BlendPose_0=@n9, BlendPose_1=@n7, bActiveValue=@n10) {blendType=Linear})
    bAlphaBoolEnabled=@n11
    Translation="0,0,0"
    Rotation="0.061,-10.212,-0.080"
    Scale="1,1,1"
   ) {
    alphaInputType=Bool
    rotationMode=BMM_Additive
    rotationSpace=BCS_BoneSpace
    scaleMode=BMM_Additive
    scaleSpace=BCS_BoneSpace
    translationMode=BMM_Additive
    translationSpace=BCS_BoneSpace
    alphaBoolBlend {blendInTime=0.5, blendOption=Linear, blendOutTime=0.5}
    boneToModify {boneName=dyn_hair_back_02}
    rotation=<rot 0.061 -10.2119 -0.08>
   }
   n13 = "Transform (Modify) Bone - Bone: dyn_hair_back_03"(ComponentPose=@n12, bAlphaBoolEnabled=@n11, Translation="0,0,0", Rotation="-0.394,-6.786,-0.862", Scale="1,1,1") {
    alphaInputType=Bool
    rotationMode=BMM_Additive
    rotationSpace=BCS_BoneSpace
    scaleMode=BMM_Additive
    scaleSpace=BCS_BoneSpace
    translationMode=BMM_Additive
    translationSpace=BCS_BoneSpace
    alphaBoolBlend {blendInTime=0.5, blendOption=Linear, blendOutTime=0.5}
    boneToModify {boneName=dyn_hair_back_03}
    rotation=<rot -0.394 -6.7859 -0.862>
   }
   n14 = "Transform (Modify) Bone - Bone: dyn_hair_back_04"(ComponentPose=@n13, bAlphaBoolEnabled=@n11, Translation="0,0,0", Rotation="3.585,3.500,-2.681", Scale="1,1,1") {
    alphaInputType=Bool
    rotationMode=BMM_Additive
    rotationSpace=BCS_BoneSpace
    scaleMode=BMM_Additive
    scaleSpace=BCS_BoneSpace
    translationMode=BMM_Additive
    translationSpace=BCS_BoneSpace
    alphaBoolBlend {blendInTime=0.5, blendOption=Linear, blendOutTime=0.5}
    boneToModify {boneName=dyn_hair_back_04}
    rotation=<rot 3.585 3.5 -2.681>
   }
   n15 = "Transform (Modify) Bone - Bone: dyn_hair_back_05"(ComponentPose=@n14, bAlphaBoolEnabled=@n11, Translation="0.068,-1.349,-0.843", Rotation="-11.052,-3.984,-18.809", Scale="1,1,1") {
    alphaInputType=Bool
    rotationMode=BMM_Additive
    rotationSpace=BCS_BoneSpace
    scaleMode=BMM_Additive
    scaleSpace=BCS_BoneSpace
    translationMode=BMM_Additive
    translationSpace=BCS_BoneSpace
    alphaBoolBlend {blendInTime=0.5, blendOption=Linear, blendOutTime=0.5}
    boneToModify {boneName=dyn_hair_back_05}
    rotation=<rot -11.052 -3.984 -18.809>
    translation=<0.068 -1.349 -0.843>
   }
   n16 = "Transform (Modify) Bone - Bone: dyn_hair_back_06"(ComponentPose=@n15, bAlphaBoolEnabled=@n11, Translation="-1.297,-0.285,0.673", Rotation="8.494,-3.675,10.225", Scale="1,1,1") {
    alphaInputType=Bool
    rotationMode=BMM_Additive
    rotationSpace=BCS_BoneSpace
    scaleMode=BMM_Additive
    scaleSpace=BCS_BoneSpace
    translationMode=BMM_Additive
    translationSpace=BCS_BoneSpace
    alphaBoolBlend {blendInTime=0.5, blendOption=Linear, blendOutTime=0.5}
    boneToModify {boneName=dyn_hair_back_06}
    rotation=<rot 8.494 -3.675 10.225>
    translation=<-1.297 -0.285 0.673>
   }
   n17 = Get Is Wearing Backpack()
   n18 = "Transform (Modify) Bone - Bone: dyn_hair_back_02"(ComponentPose=@n16, bAlphaBoolEnabled=@n17, Translation="0.484,0.002,1.097", Rotation="0,0,0", Scale="1,1,1") {
    alphaInputType=Bool
    rotationMode=BMM_Additive
    rotationSpace=BCS_BoneSpace
    scaleMode=BMM_Additive
    scaleSpace=BCS_BoneSpace
    translationMode=BMM_Additive
    translationSpace=BCS_BoneSpace
    alphaBoolBlend {blendInTime=0.1, blendOption=Linear, blendOutTime=0.1}
    boneToModify {boneName=dyn_hair_back_02}
    translation=<0.484 0.002 1.097>
   }
   n19 = "Transform (Modify) Bone - Bone: dyn_hair_back_03"(ComponentPose=@n18, bAlphaBoolEnabled=@n17, Translation="-3.348,0.000,1.334", Rotation="-7.932,5.406,0.713", Scale="1,1,1") {
    alphaInputType=Bool
    rotationMode=BMM_Additive
    rotationSpace=BCS_BoneSpace
    scaleMode=BMM_Additive
    scaleSpace=BCS_BoneSpace
    translationMode=BMM_Additive
    translationSpace=BCS_BoneSpace
    alphaBoolBlend {blendInTime=0.1, blendOption=Linear, blendOutTime=0.1}
    boneToModify {boneName=dyn_hair_back_03}
    rotation=<rot -7.932 5.406 0.713>
    translation {x=-3.348, z=1.334}
   }
   n20 = "Transform (Modify) Bone - Bone: dyn_hair_back_04"(ComponentPose=@n19, bAlphaBoolEnabled=@n17, Translation="-0.524,-0.004,-1.747", Rotation="0,0,0", Scale="1,1,1") {
    alphaInputType=Bool
    rotationMode=BMM_Additive
    rotationSpace=BCS_BoneSpace
    scaleMode=BMM_Additive
    scaleSpace=BCS_BoneSpace
    translationMode=BMM_Additive
    translationSpace=BCS_BoneSpace
    alphaBoolBlend {blendInTime=0.1, blendOption=Linear, blendOutTime=0.1}
    boneToModify {boneName=dyn_hair_back_04}
    translation=<-0.524 -0.004 -1.747>
   }
   n21 = Get Is Wearing Backpack()
   n22 = "Transform (Modify) Bone - Bone: dyn_hair_back_05"(ComponentPose=@n20, bAlphaBoolEnabled=@n21, Translation="0.301,-0.530,-0.341", Rotation="-0.167,32.717,0.043", Scale="1,1,1") {
    alphaInputType=Bool
    rotationMode=BMM_Additive
    rotationSpace=BCS_BoneSpace
    scaleMode=BMM_Additive
    scaleSpace=BCS_BoneSpace
    translationMode=BMM_Additive
    translationSpace=BCS_BoneSpace
    alphaBoolBlend {blendInTime=0.1, blendOption=Linear, blendOutTime=0.1}
    boneToModify {boneName=dyn_hair_back_05}
    rotation=<rot -0.167 32.717 0.043>
    translation=<0.301 -0.53 -0.341>
   }
   n23 = "Transform (Modify) Bone - Bone: dyn_hair_back_06"(ComponentPose=@n22, bAlphaBoolEnabled=@n21, Translation="0.068,1.193,-1.052", Rotation="0.246,77.218,0.024", Scale="1,1,1") {
    alphaInputType=Bool
    rotationMode=BMM_Additive
    rotationSpace=BCS_BoneSpace
    scaleMode=BMM_Additive
    scaleSpace=BCS_BoneSpace
    translationMode=BMM_Additive
    translationSpace=BCS_BoneSpace
    alphaBoolBlend {blendInTime=0.1, blendOption=Linear, blendOutTime=0.1}
    boneToModify {boneName=dyn_hair_back_06}
    rotation=<rot 0.246 77.218 0.024>
    translation=<0.068 1.193 -1.052>
   }
   n24 = "Transform (Modify) Bone - Bone: dyn_hair_back_05"(
    ComponentPose=@n23
    bAlphaBoolEnabled=Get pickaxebitting() -> (pickaxebitting:bool)
    Translation="-0.787,-0.978,-0.150"
    Rotation="-1.045,-38.757,-6.871"
    Scale="1,1,1"
   ) {
    alphaInputType=Bool
    rotationMode=BMM_Additive
    rotationSpace=BCS_BoneSpace
    scaleMode=BMM_Additive
    scaleSpace=BCS_BoneSpace
    translationMode=BMM_Additive
    translationSpace=BCS_BoneSpace
    alphaBoolBlend {blendInTime=0.5, blendOption=Linear, blendOutTime=0.5}
    boneToModify {boneName=dyn_hair_back_05}
    rotation=<rot -1.045 -38.757 -6.871>
    translation=<-0.787 -0.978 -0.15>
   }
   n25 = "Transform (Modify) Bone - Bone: dyn_hair_back_04"(ComponentPose=@n24, bAlphaBoolEnabled=Get bIsSkydiveFloating(), Translation="0,0,0", Rotation="2.673,26.221,-1.464", Scale="1,1,1") {
    alphaInputType=Bool
    rotationMode=BMM_Additive
    rotationSpace=BCS_BoneSpace
    scaleMode=BMM_Additive
    scaleSpace=BCS_BoneSpace
    translationMode=BMM_Additive
    translationSpace=BCS_BoneSpace
    alphaBoolBlend {blendInTime=0.5, blendOption=Linear, blendOutTime=0.5}
    boneToModify {boneName=dyn_hair_back_04}
    rotation=<rot 2.673 26.221 -1.464>
   }
   n26 = "Transform (Modify) Bone - Bone: dyn_hair_back_04"!(
    ComponentPose=@n25
    Alpha:float=Get Alpha Output() -> (Alpha Output:float)
    Translation="0,0,0"
    Rotation="3.992,15.32,-2.237"
    Scale="1,1,1"
   ) -> (Pose) {
    rotationMode=BMM_Additive
    rotationSpace=BCS_BoneSpace
    scaleMode=BMM_Additive
    scaleSpace=BCS_BoneSpace
    translationMode=BMM_Additive
    translationSpace=BCS_BoneSpace
    alphaBoolBlend {blendInTime=0.5, blendOption=Linear, blendOutTime=0.5}
    boneToModify {boneName=dyn_hair_back_04}
    rotation=<rot 3.992 15.32 -2.237>
   }
   Output Pose(Result=Component To Local(ComponentPose=@n26)) {layerGroup=None}
  ]}
 EventGraph {nodes [
   Event BlueprintUpdateAnimation()
   n1 = CheckForBackbling()
   > Set Is Backbling A Cape(Is Backbling A Cape:bool=@n1.bIsBackblingACape) -> (Output_Get:bool)
   > Set Is Wearing Backpack(Is Wearing Backpack:bool=AND Boolean(A=@n1, B=NOT Boolean(A=@n1.bIsBackblingACape))) -> (Output_Get:bool)
   > Branch(Condition=Get Is Wearing Backpack())
   n2 = GetAnimInstance(self=GetOwnerSkeletalMeshComponent())
   > Set pickaxebitting(pickaxebitting:bool="Equal (Object)"(
     A=GetCurrentActiveMontage!(self:Anim Instance Object Reference=@n2) -> (ReturnValue:Anim Montage Object Reference)
     B=/Game/Animation/Game/MainPlayer/Montages/Melee/PlayerMeleeCombo
    ) {is=title K2Node_CallFunction}) -> (Output_Get:bool)
   > Set Alpha Output(Alpha Output:float=GetAlphaFromReferenceTransformDelta(SkeletalMeshComponent=GetOwnerSkeletalMeshComponent(), BoneName=head, BaseName=spine_05, InputRange=(X=0,Y=-40), TwistAxis="1,0,0")) -> (Output_Get:float)
  ]}
]
variables [
 Alpha Output {type=number}
 Is Backbling A Cape {type=boolean}
 Is Wearing Backpack {type=boolean}
 pickaxebitting {type=boolean}
]
